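; Game state component that gives pre-chromed containers their chrome loot behavior at match start.
; On authority, once playlist data is ready and the hotfix allows it, it listens for the game phase to pass Setup.
; It then scans building containers once in delayed batches, applying a gameplay effect to chrome-tagged ones with a valid override.
; Overrides come from a dedicated-server DataTable lookup matching loot tier group and tag query, returning the chrome tier group.
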(event EventBeginPlay
  ;[0.02556 0.1615 0.04016] BeginPlay: On match start, find all the pre-chromed containers and update them to their respective chrome loot.
  (|CallParentFunction)
  (Networking|SwitchHasAuthority
    (:Authority
      (Class|FortAsyncActionCurrentPlaylistReady|PlaylistDataReadyAsync
        (:OnReady
          (bind _output_get (Variables|Default|SetbAllowChromeChestGameplay (ScalableFloat|GetValueasBool (Variables|Default|GetHotfixAllowChromeChestGameplay))))
          (if _output_get
            (Fortnite|FortShippingLog self "BeginPlay: Applying chrome chest gameplay to the pre-chromed chests.")
            (bind _returnvalue (GameplayEventRouter|GetEventRouter "/Script/FortniteGame.FortGameState" (Components|GetOwner)))
            (bind (_asynctaskproxy _payload) (GameplayEventRouter|StartListeningToStatefulEvent _returnvalue "/Script/FortniteGame.FortGamePhaseUpdatedEvent")
              (:then
                (Variables|Default|SetGamePhaseUpdatedAsyncTask _asynctaskproxy))
              (:OnEventReceived
                _)
              (:OnSavedEventStateExists
                _))
            (CallFunction|HandleGamePhaseChanged (Utilities|Struct|BreakFortGamePhaseUpdatedEvent _payload))
            (else
              (Fortnite|FortShippingLog self "BeginPlay: ChromeChestGameplay is hotfixed off."))))))))

(event Custom|HandleGamePhaseChanged
  (params (GamePhase :type EAthenaGamePhase%20Enum))
  ;[0.03936 0.2812 0.2362] HandleGamePhaseChanged: Apply pre chrome chest gameplay effect once the match has started.
  (Networking|SwitchHasAuthority
    (:Authority
      (if (not (or (Utilities|Enum|Equal(Enum) GamePhase "None") (Utilities|Enum|Equal(Enum) GamePhase "Setup")))
        (CallFunction|ApplyChromeChestGameplayEffect)
        (bind _gamephaseupdatedasynctask (Variables|Default|GetGamePhaseUpdatedAsyncTask)
          (:then
            (GameplayEventRouter|StopListeningtoStatefulEvent _gamephaseupdatedasynctask))
          (:else))))))

(event Custom|ApplyChromeChestGameplayEffect
  ;[0 0.2083 0.02013] ApplyChromeChestGameplayEffect: Find all the pre-chrome chest in the world and add the chrome chest gameplay effect, for gameplay (loot tier change).
  (Networking|SwitchHasAuthority
    (:Authority
      (Utilities|FlowControl|DoOnce false
        (:Completed
          (bind _outactors (Actor|GetAllActorsOfClass "/Script/FortniteGame.BuildingContainer"))
          (bind _outelementarray (Default|GetSubsetOfArraywithDelayLoop _outactors 20)
            (:Loop
              (for _array_element _outelementarray
                (Utilities|IsValid _array_element
                  (:Is%20Valid
                    (if (Actor|ActorHasTag _array_element "Gameplay.Status.Chrome")
                      (bind _bvalid (CallFunction|GetChromeChestOverrideInfo _array_element))
                      (if _bvalid
                        (Utilities|IsValid (Ability|GetAbilitySystemComponent _array_element)
                          (:Is%20Valid
                            (GameplayEffects|ApplyGameplayEffectToSelf (Ability|GetAbilitySystemComponent _array_element) (Variables|Default|GetChromeChestGameplayEffect) 0.0 (GameplayEffects|MakeEffectContext (Ability|GetAbilitySystemComponent _array_element)))))))))))))))))

(event EventEndPlay
  (params (EndPlayReason :type EEndPlayReason%20Enum :default "Destroyed"))
  ;[0.1302 0.07111 0.008553] EndPlay: Unbind delegates.
  (|CallParentFunction EndPlayReason)
  (Networking|SwitchHasAuthority
    (:Authority
      (bind _gamephaseupdatedasynctask (Variables|Default|GetGamePhaseUpdatedAsyncTask)
        (:then
          (GameplayEventRouter|StopListeningtoStatefulEvent _gamephaseupdatedasynctask))
        (:else)))))

(fn GetChromeChestOverrideInfo
  (params (Container :type Building%20Container%20Object%20Reference))
  (bind _chromechestoverridedatatable (Variables|Default|GetChromeChestOverrideDatatable))
  ;[0.06655 0.0331 0.08333] Returns the chrome loot tier group for a particular container.
  (Utilities|FlowControl|Sequence
    (:then_0
      (Utilities|IsValid Container
        (:Is%20Valid
          (if (and (Networking|IsDedicatedServer) (Variables|Default|GetbAllowChromeChestGameplay))
            (bind _outrownames (DataTable|GetDataTableRowNames _chromechestoverridedatatable))
            (for _array_element _outrownames
              (bind _returnvalue (Utilities|GetDataTableRow _chromechestoverridedatatable _array_element))
              (if (Math|Boolean|ANDBoolean (Utilities|Struct|BreakChromeChestLootOverrideStruct _returnvalue) (== (.OriginalLootTierGroup_8_EA8EF57D480C415763DCB881260885EC (Utilities|Struct|BreakChromeChestLootOverrideStruct _returnvalue)) (Variables|/Script/FortniteGame.BuildingContainer|GetSearchLootTierGroup Container)) (GameplayTags|DoesContainerMatchTagQuery (GameplayTags|GetOwnedGameplayTags Container :!class /Script/GameplayTags.GameplayTagAssetInterface) (.TagQuery_4_21197CE24EDD0504A4D62CB8C468DDE4 (Utilities|Struct|BreakChromeChestLootOverrideStruct _returnvalue))))
                (return true (.ChromeLootTierGroup_12_8C268813457EB3591EBC64B22F97E95A (Utilities|Struct|BreakChromeChestLootOverrideStruct _returnvalue)))))))))
    (:then_1
      (return false "None"))))
